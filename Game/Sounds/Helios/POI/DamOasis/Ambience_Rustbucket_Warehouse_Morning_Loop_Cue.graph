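# Looping morning ambience cue for the Rustbucket warehouse at the Dam Oasis POI, using indoor ambient sound class.
# The first two quality branches layer a randomly chosen quad industrial bed and desert wind with randomly chosen distant metal rattling.
# The last two use a lighter stereo mix of desert day and industrial morning.
# Looping envelopers slowly drift volume and pitch, with strong swells on the rattle, while modulators add random variation.
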SoundCue /Game/Sounds/Helios/POI/DamOasis/Ambience_Rustbucket_Warehouse_Morning_Loop_Cue
root=QualityLevel_0
nodes [
 {
  id=Enveloper_0
  children [Enveloper_3]
  properties {
   bLoop=true
   bLoopIndefinitely=true
   loopEnd=100
   pitchCurve {editorCurveData {keys [<at 0 0.8>]}}
   volumeCurve {editorCurveData {keys [<at 0 1 RCIM_Cubic>,<at 50 0.9 RCIM_Cubic>,<at 100 1 RCIM_Cubic>]}}
  }
 }
 {
  id=Enveloper_1
  children [Mixer_1]
  properties {
   bLoop=true
   bLoopIndefinitely=true
   loopEnd=30
   pitchCurve {editorCurveData {keys [{interpMode=RCIM_Cubic, time=0}, <at 15.9468 0.9181 RCIM_Cubic>, {interpMode=RCIM_Cubic, time=30}]}}
   volumeCurve {editorCurveData {keys [<at 0 1 RCIM_Cubic>,<at 16.1421 0.921 RCIM_Cubic>,<at 30 1 RCIM_Cubic>]}}
  }
 }
 {
  id=Enveloper_2
  children [Enveloper_4]
  properties {
   bLoop=true
   bLoopIndefinitely=true
   loopEnd=40
   pitchCurve {editorCurveData {keys [{time=0}]}}
   volumeCurve {editorCurveData {keys [
      {arriveTangent=-0.01, interpMode=RCIM_Cubic, time=0, value=1}
      {arriveTangent=-0.0292, interpMode=RCIM_Cubic, leaveTangent=-0.0292, time=10.0549, value=0.3606}
      {arriveTangent=-0.0105, interpMode=RCIM_Cubic, leaveTangent=-0.0105, tangentMode=RCTM_User, time=17.8246, value=0.4801}
      {arriveTangent=0.0314, interpMode=RCIM_Cubic, leaveTangent=0.0314, time=23.5954, value=0.2948}
      {arriveTangent=0.0013, interpMode=RCIM_Cubic, leaveTangent=0.0013, tangentMode=RCTM_User, time=28, value=0.8}
      {arriveTangent=0.0167, interpMode=RCIM_Cubic, leaveTangent=0.0167, time=33.0549, value=0.3606}
      <at 40 1>
     ]}}
  }
 }
 {
  id=Enveloper_3
  children [Mixer_0]
  properties {
   bLoop=true
   bLoopIndefinitely=true
   loopEnd=30
   pitchCurve {editorCurveData {keys [{interpMode=RCIM_Cubic, time=0}, <at 15.9468 0.9181 RCIM_Cubic>, {interpMode=RCIM_Cubic, time=30}]}}
   volumeCurve {editorCurveData {keys [<at 0 1>]}}
  }
 }
 {
  id=Enveloper_4
  children [Modulator_0]
  properties {
   bLoop=true
   bLoopIndefinitely=true
   loopEnd=30
   pitchCurve {editorCurveData {keys [{interpMode=RCIM_Cubic, time=0}, <at 15.9468 0.9181 RCIM_Cubic>, {interpMode=RCIM_Cubic, time=30}]}}
   volumeCurve {editorCurveData {keys [<at 0 1>]}}
  }
 }
 {id=Mixer_0, children [Modulator_11,WavePlayer_0], properties {inputVolume [1.2,0.3]}}
 {id=Mixer_1, children [WavePlayer_9,WavePlayer_6], properties {inputVolume [1,0.5]}}
 {id=Mixer_2, children [Enveloper_0,Enveloper_2], properties {inputVolume [1,1]}}
 {id=Modulator_0, children [Random_0], properties {pitchMax=0.9, pitchMin=0.8}}
 {id=Modulator_11, children [Random_1], properties {pitchMax=1, pitchMin=0.97}}
 {id=Modulator_2, children [Enveloper_1], properties {pitchMax=1.1, pitchMin=0.9, volumeMax=1.1, volumeMin=0.9}}
 {id=QualityLevel_0, children [Mixer_2,Mixer_2,Modulator_2,Modulator_2]}
 {id=Random_0, children [WavePlayer_8, WavePlayer_10], properties {weights [1,1]}}
 {id=Random_1, children [WavePlayer_1,WavePlayer_2,WavePlayer_3,WavePlayer_4,WavePlayer_5], properties {weights [1,1,1,1,1]}}
 {id=WavePlayer_0, properties {soundWaveAssetPtr=&/Ambience_Desert_Valley_Wind_Morning_Ext_Loop_Quad}}
 {id=WavePlayer_1, properties {bLooping=true, soundWaveAssetPtr=/Game/Sounds/Asteria/POI/IndustrialTown/Ambience_Industrial_Morning_01_Quad}}
 {
  id=WavePlayer_10
  properties {bLooping=true, soundWaveAssetPtr=/Game/Sounds/Asteria/POI/JunkYard/Ambience_Metal_Distant_Rattling_In_Wind_Loop_02_Quad}
 }
 {id=WavePlayer_2, properties {bLooping=true, soundWaveAssetPtr=/Game/Sounds/Asteria/POI/IndustrialTown/Ambience_Industrial_Morning_02_Quad}}
 {id=WavePlayer_3, properties {bLooping=true, soundWaveAssetPtr=/Game/Sounds/Asteria/POI/IndustrialTown/Ambience_Industrial_Morning_03_Quad}}
 {id=WavePlayer_4, properties {bLooping=true, soundWaveAssetPtr=/Game/Sounds/Asteria/POI/IndustrialTown/Ambience_Industrial_Morning_04_Quad}}
 {id=WavePlayer_5, properties {bLooping=true, soundWaveAssetPtr=/Game/Sounds/Asteria/POI/IndustrialTown/Ambience_Industrial_Morning_05_Quad}}
 {id=WavePlayer_6, properties {bLooping=true, soundWaveAssetPtr=/Game/Sounds/Asteria/POI/IndustrialTown/Ambience_Industrial_Morning_03_Stereo}}
 {
  id=WavePlayer_8
  properties {bLooping=true, soundWaveAssetPtr=/Game/Sounds/Asteria/POI/JunkYard/Ambience_Metal_Distant_Rattling_In_Wind_Loop_01_Quad}
 }
 {id=WavePlayer_9, properties {bLooping=true, soundWaveAssetPtr=/Game/Sounds/Artemis/Biomes/Desert/Ambience_Desert_Day_Stereo}}
]
properties {
 attenuationSettings=&Fort_Ambient_Athena_Att
 maxDistance=2097152
 soundClassObject=&Fort_Ambient_Indoors_Athena
 volumeMultiplier=0.301
}
