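# Looping night ambience cue for the Rustbucket warehouse at the DamOasis POI, using indoor ambient sound class.
# A quality switch gives the first two levels a bed of haunted-house, desert-wind and industrial loops plus randomly chosen, pitched-down metal rattles.
# The last two levels use a simpler desert and industrial mix with random pitch and volume.
# Endlessly looping envelopes, nested with differing lengths in the richer branch, slowly vary volume and pitch, making rattles swell and fade.
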SoundCue /Game/Sounds/Helios/POI/DamOasis/Ambience_Rustbucket_Warehouse_Night_Loop_Cue
root=QualityLevel_0
nodes [
 {
  id=Enveloper_0
  children [Enveloper_3]
  properties {
   bLoop=true
   bLoopIndefinitely=true
   loopEnd=100
   pitchCurve {editorCurveData {keys [<at 0 0.8>]}}
   volumeCurve {editorCurveData {keys [<at 0 1 RCIM_Cubic>,<at 50 0.9 RCIM_Cubic>,<at 100 1 RCIM_Cubic>]}}
  }
 }
 {
  id=Enveloper_1
  children [Mixer_0]
  properties {
   bLoop=true
   bLoopIndefinitely=true
   loopEnd=30
   pitchCurve {editorCurveData {keys [{interpMode=RCIM_Cubic, time=0}, <at 15.9468 0.9181 RCIM_Cubic>, {interpMode=RCIM_Cubic, time=30}]}}
   volumeCurve {editorCurveData {keys [<at 0 1 RCIM_Cubic>,<at 16.1421 0.921 RCIM_Cubic>,<at 30 1 RCIM_Cubic>]}}
  }
 }
 {
  id=Enveloper_2
  children [Modulator_0]
  properties {
   bLoop=true
   bLoopIndefinitely=true
   loopEnd=30
   pitchCurve {editorCurveData {keys [{interpMode=RCIM_Cubic, time=0}, <at 15.9468 0.9181 RCIM_Cubic>, {interpMode=RCIM_Cubic, time=30}]}}
   volumeCurve {editorCurveData {keys [<at 0 1>]}}
  }
 }
 {
  id=Enveloper_3
  children [Mixer_3]
  properties {
   bLoop=true
   bLoopIndefinitely=true
   loopEnd=30
   pitchCurve {editorCurveData {keys [{interpMode=RCIM_Cubic, time=0}, <at 15.9468 0.9181 RCIM_Cubic>, {interpMode=RCIM_Cubic, time=30}]}}
   volumeCurve {editorCurveData {keys [<at 0 1>]}}
  }
 }
 {
  id=Enveloper_4
  children [Enveloper_2]
  properties {
   bLoop=true
   bLoopIndefinitely=true
   loopEnd=40
   pitchCurve {editorCurveData {keys [{time=0}]}}
   volumeCurve {editorCurveData {keys [
      {arriveTangent=-0.01, interpMode=RCIM_Cubic, time=0, value=1}
      <at 10.0039 0.2703 RCIM_Cubic>
      {arriveTangent=-0.0105, interpMode=RCIM_Cubic, leaveTangent=-0.0105, tangentMode=RCTM_User, time=20, value=1}
      {arriveTangent=-0.025, interpMode=RCIM_Cubic, leaveTangent=-0.025, time=23.5545, value=0.0703}
      {arriveTangent=0.0013, interpMode=RCIM_Cubic, leaveTangent=0.0013, tangentMode=RCTM_User, time=28, value=0.8}
      {arriveTangent=0.0167, interpMode=RCIM_Cubic, leaveTangent=0.0167, time=33.0039, value=0.2703}
      <at 40 1>
     ]}}
  }
 }
 {id=Mixer_0, children [WavePlayer_2,WavePlayer_0], properties {inputVolume [0.5,1]}}
 {id=Mixer_2, children [Enveloper_0,Enveloper_4], properties {inputVolume [1,1]}}
 {id=Mixer_3, children [WavePlayer_5, WavePlayer_4, WavePlayer_7, WavePlayer_12], properties {inputVolume [1,0.4,0.4,0.5]}}
 {id=Modulator_0, children [Random_1], properties {pitchMax=0.9, pitchMin=0.8}}
 {id=Modulator_2, children [Enveloper_1], properties {pitchMax=1.1, pitchMin=0.9, volumeMax=1.1, volumeMin=0.9}}
 {id=QualityLevel_0, children [Mixer_2,Mixer_2,Modulator_2,Modulator_2]}
 {id=Random_1, children [WavePlayer_3,WavePlayer_6], properties {weights [1,1]}}
 {id=WavePlayer_0, properties {bLooping=true, soundWaveAssetPtr=/Game/Sounds/Asteria/POI/IndustrialTown/Ambience_Industrial_Night_04_Quad}}
 {
  id=WavePlayer_12
  properties {bLooping=true, soundWaveAssetPtr=/Game/Sounds/Asteria/POI/IcebergIsles_ResearchStation/Ambience_IndustrialLarge001_Loop_Quad}
 }
 {id=WavePlayer_2, properties {bLooping=true, soundWaveAssetPtr=/Game/Sounds/Artemis/Biomes/Desert/Ambience_Desert_Night_Stereo}}
 {
  id=WavePlayer_3
  properties {bLooping=true, soundWaveAssetPtr=/Game/Sounds/Asteria/POI/JunkYard/Ambience_Metal_Distant_Rattling_In_Wind_Loop_03_Quad}
 }
 {
  id=WavePlayer_4
  properties {bLooping=true, soundWaveAssetPtr=&/Ambience_Desert_Valley_Wind_Night_Ext_Loop_Quad}
 }
 {
  id=WavePlayer_5
  properties {bLooping=true, soundWaveAssetPtr=/Game/Sounds/Artemis/POI/HauntedHouse/HauntedHouse_AmbientLoop_Exterior_Main_02}
 }
 {
  id=WavePlayer_6
  properties {bLooping=true, soundWaveAssetPtr=/Game/Sounds/Asteria/POI/JunkYard/Ambience_Metal_Distant_Rattling_In_Wind_Loop_04_Quad}
 }
 {id=WavePlayer_7, properties {bLooping=true, soundWaveAssetPtr=&Ambience_Industrial_Night_04_Quad}}
]
properties {
 attenuationSettings=&Fort_Ambient_Athena_Att
 maxDistance=2097152
 soundClassObject=&Fort_Ambient_Indoors_Athena
 volumeMultiplier=0.175
}
